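# Deformation Control Rig for the VoidSpray female medium body cosmetic, correcting body deformation and anchoring outfit accessory bones.
# A sequence runs the shared DeformRig function, then weighted parent constraints on zipper, strap, jacket, cans, karabiner and belt, with one can aimed at the other.
# Spherical pose readers then drive corrective offsets on both upper-arm deform bones and the belt.
# The second back-jacket constraint targets the left bone with right-side parents, leaving the right unconstrained.
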ControlRigBlueprint /BRCosmetics/Characters/Player/Female/Medium/Bodies/F_MED_VoidSpray/Meshes/F_MED_VoidSpray_CtrlRig_Deform
graphs [
 RigVMFunctionLibrary {links [], nodes []}
 RigVMFunction_Sequence_1_ContainedGraph {nodes [
   n1 = RigVMFunctionEntryNode()
   RigVMFunction_Sequence = &RigVMFunction_Sequence(ExecuteContext=@n1.ExecuteContext)
   n2 = &RigVMFunction_Sequence(ExecuteContext=@RigVMFunction_Sequence.B)
   n3 = &RigVMFunction_Sequence(ExecuteContext=@n2.B)
   RigVMFunctionReturnNode(A=@RigVMFunction_Sequence.A, B=@n2.A, G=@n3.A, H=@n3.B)
  ]}
 RigVMModel {nodes [
   RigUnit_BeginExecution = &RigUnit_BeginExecution()
   RigVMFunction_Sequence = RigVMAggregateNode(ExecuteContext=@RigUnit_BeginExecution.ExecutePin)
   FN_DeformRig_FNC_1 = RigVMFunctionReferenceNode(
    DeformRig_ Curve="(deformRig_l_upperarm_up_alpha_72_#1=False,deformRig_r_upperarm_up_alpha_73_#2=False,deformRig_l_upperarm_fr_alpha_74_#3=False,deformRig_r_upperarm_fr_alpha_75_CB51B8 ...(truncated)"
    DeformRig_Bool="(upperarm_1_#4=True,elbow_3_#5=True,wrist_7_#6=True,groin_8_#7=True,glute_17_#8=True,knee_9_40C408324CA261 ...(truncated)"
    DeformRig_Debug=(upperarm_debug_18_#4=False,wrist_debug_20_#6=False,lat_debug_26_#9=False,groinGlute_debug_27_#7=False)
    DeformRig_Value="(upperarm_up_18_#10=(Rotation=(X=0,Y=0,Z=0,W=1),Translation=(X=0,Y=0,Z=8),Scale3D=(X=0.7,Y=1,Z=1)),upperarm_fwd_fr_19_#11=(Rotation=(X=0,Y=0,Z=0,W=1),Translation=(X=0,Y=10,Z=0),Scale3D=(X=0.5,Y=1,Z= ...(truncated)"
    ExecuteContext=@RigVMFunction_Sequence.A
   )
   &RigUnit_ParentConstraint(
    AdvancedSettings=(InterpolationType=Average,RotationOrderForFilter=XZY)
    Child=(Type=Bone,Name="dyn_zipper_MSTR")
    Parents=((Item=(Type=Bone,Name="spine_04"),Weight=0.382),(Item=(Type=Bone,Name="spine_05"),Weight=0.173),(Item=(Type=Bone,Name="deform_pec_r"),Weight=0.418),(Item=(Type=Bone,Name="clavicle_r"),Weight=0.028))
    Weight=1
    ExecutePin=@RigVMFunction_Sequence.B
   )
   > &RigUnit_ParentConstraint(
    AdvancedSettings=(InterpolationType=Average,RotationOrderForFilter=XZY)
    Child=(Type=Bone,Name="dyn_backstrap_MSTR")
    Parents=((Item=(Type=Bone,Name="spine_04"),Weight=0.14),(Item=(Type=Bone,Name="spine_05"),Weight=0.86))
    Weight=1
   )
   > &RigUnit_ParentConstraint(
    AdvancedSettings=(InterpolationType=Average,RotationOrderForFilter=XZY)
    Child=(Type=Bone,Name="dyn_side_jacket_MSTR_l")
    Parents=((Item=(Type=Bone,Name="spine_02"),Weight=0.093),(Item=(Type=Bone,Name="spine_03"),Weight=0.098),(Item=(Type=Bone,Name="spine_04"),Weight=0.742),(Item=(Type=Bone,Name="spine_05"),Weight=0.066))
    Weight=1
   )
   > &RigUnit_ParentConstraint(
    AdvancedSettings=(InterpolationType=Average,RotationOrderForFilter=XZY)
    Child=(Type=Bone,Name="dyn_side_jacket_MSTR_r")
    Parents=((Item=(Type=Bone,Name="spine_02"),Weight=0.093),(Item=(Type=Bone,Name="spine_03"),Weight=0.098),(Item=(Type=Bone,Name="spine_04"),Weight=0.742),(Item=(Type=Bone,Name="spine_05"),Weight=0.066))
    Weight=1
   )
   > &RigUnit_ParentConstraint(
    AdvancedSettings=(InterpolationType=Average,RotationOrderForFilter=XZY)
    Child=(Type=Bone,Name="dyn_back_low_jacket_MSTR_l")
    Parents=((Item=(Type=Bone,Name="spine_02"),Weight=0.133),(Item=(Type=Bone,Name="spine_04"),Weight=0.473),(Item=(Type=Bone,Name="spine_05"),Weight=0.128),(Item=(Type=Bone,Name="clavicle_l"),Weight=0.159),(Item=(Type=Bone,Name="upperarm_twist_01_l"),Weight=0.107))
    Weight=1
   )
   > &RigUnit_ParentConstraint(
    AdvancedSettings=(InterpolationType=Average,RotationOrderForFilter=XZY)
    Child=(Type=Bone,Name="dyn_back_low_jacket_MSTR_l")
    Parents=((Item=(Type=Bone,Name="spine_02"),Weight=0.133),(Item=(Type=Bone,Name="spine_04"),Weight=0.473),(Item=(Type=Bone,Name="spine_05"),Weight=0.128),(Item=(Type=Bone,Name="clavicle_r"),Weight=0.159),(Item=(Type=Bone,Name="upperarm_twist_01_r"),Weight=0.107))
    Weight=1
   )
   > &RigUnit_ParentConstraint(
    AdvancedSettings=(InterpolationType=Average,RotationOrderForFilter=XZY)
    Child=(Type=Bone,Name="dyn_back_low_jacket_MSTR")
    Parents=((Item=(Type=Bone,Name="spine_02"),Weight=0.108),(Item=(Type=Bone,Name="spine_03"),Weight=0.121),(Item=(Type=Bone,Name="spine_04"),Weight=0.418),(Item=(Type=Bone,Name="spine_05"),Weight=0.354))
    Weight=1
   )
   > &RigUnit_ParentConstraint(
    AdvancedSettings=(InterpolationType=Average,RotationOrderForFilter=XZY)
    Child=(Type=Bone,Name="extra_can_01")
    Filter=(TranslationFilter=(bX=True,bY=True,bZ=True),RotationFilter=(bX=true,bY=true,bZ=false),ScaleFilter=(bX=True,bY=True,bZ=True))
    Parents=((Item=(Type=Bone,Name="pelvis"),Weight=0.752),(Item=(Type=Bone,Name="spine_01"),Weight=0.058),(Item=(Type=Bone,Name="thigh_twist_01_l"),Weight=0.094),(Item=(Type=Bone,Name="deform_glute_l"),Weight=0.096))
    Weight=1
   )
   > &RigUnit_ParentConstraint(
    AdvancedSettings=(InterpolationType=Average,RotationOrderForFilter=XZY)
    Child=(Type=Bone,Name="extra_can_02")
    Filter=(TranslationFilter=(bX=True,bY=True,bZ=True),RotationFilter=(bX=true,bY=true,bZ=true),ScaleFilter=(bX=True,bY=True,bZ=True))
    Parents=((Item=(Type=Bone,Name="pelvis"),Weight=0.311),(Item=(Type=Bone,Name="thigh_l"),Weight=0.023),(Item=(Type=Bone,Name="thigh_twist_01_l"),Weight=0.53),(Item=(Type=Bone,Name="deform_glute_l"),Weight=0.136))
    Weight=1
   )
   > &RigUnit_AimConstraintLocalSpaceOffset(
    Child=(Type=Bone,Name="extra_can_01")
    Filter=(bX=false,bY=true,bZ=True)
    Parents=((Item=(Type=Bone,Name="extra_can_02"),Weight=1))
    Weight=1
    WorldUp=(Target=(X=0,Y=0,Z=1),Kind=Direction,Space=(Type=Bone,Name="extra_can_01"))
   )
   > &RigUnit_ParentConstraint(
    AdvancedSettings=(InterpolationType=Average,RotationOrderForFilter=XZY)
    Child=(Type=Bone,Name="dyn_karabin_MSTR")
    Filter=(TranslationFilter=(bX=True,bY=True,bZ=True),RotationFilter=(bX=true,bY=true,bZ=true),ScaleFilter=(bX=True,bY=True,bZ=True))
    Parents=((Item=(Type=Bone,Name="pelvis"),Weight=0.047),(Item=(Type=Bone,Name="thigh_l"),Weight=0.269),(Item=(Type=Bone,Name="thigh_twist_01_l"),Weight=0.66),(Item=(Type=Bone,Name="deform_glute_l"),Weight=0.023))
    Weight=1
   )
   > &RigUnit_ParentConstraint(
    AdvancedSettings=(InterpolationType=Average,RotationOrderForFilter=XZY)
    Child=(Type=Bone,Name="dyn_belt_MSTR")
    Parents=((Item=(Type=Bone,Name="pelvis"),Weight=0.854),(Item=(Type=Bone,Name="spine_01"),Weight=0.087),(Item=(Type=Bone,Name="spine_02"),Weight=0.058))
    Weight=1
   )
   n1 = &RigVMFunction_Sequence(ExecuteContext=@RigVMFunction_Sequence.G)
   n2 = &RigUnit_SphericalPoseReader(
    ActiveRegionScaleFactors=(PositiveWidth=0.5,NegativeWidth=1,PositiveHeight=1,NegativeHeight=1)
    ActiveRegionSize=0.2
    Debug=(bDrawDebug=false,bDraw2D=False,bDrawLocalAxes=False,DebugScale=25,DebugSegments=20,DebugThickness=0.25)
    DriverAxis=(X=1,Y=0,Z=0)
    DriverItem=(Type=Bone,Name="upperarm_l")
    FalloffRegionScaleFactors=(PositiveWidth=0.5,NegativeWidth=1,PositiveHeight=1,NegativeHeight=1)
    FalloffSize=0.35
    FlipHeightScaling=False
    FlipWidthScaling=False
    OptionalParentItem=(Type=Bone,Name="spine_05")
    RotationOffset=(X=0,Y=0,Z=0)
    ExecutePin=@n1.A
   )
   > /Script/ControlRig.RigUnit_SetRelativeTransformForItem(
    Child=(Type=Bone,Name="deform_upperarm_up_l")
    Parent=(Type=Bone,Name="deform_upperarm_up_l")
    Value=(Rotation=(X=0,Y=0.2588,Z=0,W=0.9659),Translation=(X=4,Y=0,Z=-2),Scale3D=(X=1,Y=1,Z=1))
    bParentInitial=false
    bPropagateToChildren=true
    Weight=@n2.OutputParam
   ) {
    is=struct RigVMUnitNode
    template="Set Relative Transform::Execute(in Child,in Parent,in bParentInitial,in Value,in Weight,in bPropagateToChildren)"
   }
   n3 = &RigUnit_SphericalPoseReader(
    ActiveRegionScaleFactors=(PositiveWidth=0.5,NegativeWidth=1,PositiveHeight=1,NegativeHeight=1)
    ActiveRegionSize=0.2
    Debug=(bDrawDebug=false,bDraw2D=False,bDrawLocalAxes=False,DebugScale=25,DebugSegments=20,DebugThickness=0.25)
    DriverAxis=(X=1,Y=0,Z=0)
    DriverItem=(Type=Bone,Name="upperarm_r")
    FalloffRegionScaleFactors=(PositiveWidth=0.5,NegativeWidth=1,PositiveHeight=1,NegativeHeight=1)
    FalloffSize=0.35
    FlipHeightScaling=False
    FlipWidthScaling=False
    OptionalParentItem=(Type=Bone,Name="spine_05")
    RotationOffset=(X=0,Y=0,Z=0)
    ExecutePin=@n1.B
   )
   > &RigUnit_SetRelativeTransformForItem(
    Child=(Type=Bone,Name="deform_upperarm_up_r")
    Parent=(Type=Bone,Name="deform_upperarm_up_r")
    Value=(Rotation=(X=0,Y=0.2588,Z=0,W=0.9659),Translation=(X=-4,Y=0,Z=2),Scale3D=(X=1,Y=1,Z=1))
    bParentInitial=false
    bPropagateToChildren=true
    Weight=@n3.OutputParam
   ) {template="Set Relative Transform::Execute(in Child,in Parent,in bParentInitial,in Value,in Weight,in bPropagateToChildren)"}
   n4 = &RigUnit_SphericalPoseReader(
    ActiveRegionScaleFactors=(PositiveWidth=1,NegativeWidth=1,PositiveHeight=1,NegativeHeight=0.7)
    ActiveRegionSize=0.2
    Debug=(bDrawDebug=true,bDraw2D=false,bDrawLocalAxes=False,DebugScale=25,DebugSegments=20,DebugThickness=0.25)
    DriverAxis=(X=-1,Y=0,Z=0)
    DriverItem=(Type=Bone,Name="thigh_r")
    FalloffRegionScaleFactors=(PositiveWidth=0.5,NegativeWidth=0.9,PositiveHeight=0.6,NegativeHeight=0.8)
    FalloffSize=0.5
    FlipHeightScaling=False
    FlipWidthScaling=False
    OptionalParentItem=(Type=Bone,Name="pelvis")
    RotationOffset=(X=100,Y=0,Z=-20)
    ExecutePin=@RigVMFunction_Sequence.H
   )
   > &RigUnit_SetRelativeTransformForItem(
    Child=(Type=Bone,Name="dyn_belt_MSTR")
    Parent=(Type=Bone,Name="dyn_belt_MSTR")
    Value=(Rotation=(X=-0.2271,Y=-0.3564,Z=-0.487,W=0.7644),Translation=(X=0,Y=0,Z=0),Scale3D=(X=1,Y=1,Z=1))
    bParentInitial=false
    bPropagateToChildren=true
    Weight=@n4.OutputParam
   ) {template="Set Relative Transform::Execute(in Child,in Parent,in bParentInitial,in Value,in Weight,in bPropagateToChildren)"}
  ]}
]
hierarchy [
 root {loc=<0 0 0>, rot=<0 0 0 1>}
 pelvis {parent=root, loc=<0 -0.5301 98.0571>, rot=<0.0127 0.707 -0.0127 -0.707>}
 spine_01 {parent=pelvis, loc=<6.1208 -0.9328 0>, rot=<0 0 0.034 -0.9994>}
 spine_02 {parent=spine_01, loc=<6.5997 -0.2095 0>, rot=<0 0 0.0238 -0.9997>}
 spine_03 {parent=spine_02, loc=<6.4914 -0.1237 0>, rot=<0 0 0.0149 -0.9999>}
 spine_04 {parent=spine_03, loc=<6.9598 0.0794 0>, rot=<0 0 -0.0132 -0.9999>}
 spine_05 {parent=spine_04, loc=<6.9987 0 0>, rot=<0 0 -0.0015 -1>}
 neck_01 {parent=spine_05, loc=<11.3138 0.1529 0>, rot=<0 0 -0.2003 -0.9797>}
 neck_02 {parent=neck_01, loc=<4.5627 0 0>, rot=<0 0 0.0504 -0.9987>}
 head {parent=neck_02, loc=<5.3914 0 0>, rot=<0 0 0.115 -0.9934>}
 FACIAL_C_FacialRoot {parent=head, loc=<1.7595 0.334 0>, rot=<0.4877 -0.5121 -0.4876 -0.5121>}
 FACIAL_C_UnderChin {parent=FACIAL_C_FacialRoot, loc=<-0.0008 8.241 3.1468>, rot=<0.5025 0.0001 0 -0.8646>}
 FACIAL_L_UnderChin {parent=FACIAL_C_FacialRoot, loc=<2.8531 6.983 2.1735>, rot=<0.4016 -0.355 0.0819 -0.8402>}
 FACIAL_R_UnderChin {parent=FACIAL_C_FacialRoot, loc=<-2.8519 6.983 2.1726>, rot=<0.4016 0.3551 -0.082 -0.8402>}
 FACIAL_C_Jaw {parent=FACIAL_C_FacialRoot, loc=<-0.0003 2.2337 0.5264>, rot=<0.3413 0 0 -0.94>}
 FACIAL_C_Jawline {parent=FACIAL_C_Jaw, loc=<-0.0005 1.9809 7.3897>, rot=<0.2601 -0.0001 0 -0.9656>}
 FACIAL_L_Jawline {parent=FACIAL_C_FacialRoot, loc=<4.0476 5.3156 3.4349>, rot=<0.2051 0.158 -0.054 -0.9644>}
 FACIAL_R_Jawline {parent=FACIAL_C_FacialRoot, loc=<-4.0479 5.3151 3.4332>, rot=<0.2051 -0.1579 0.0539 -0.9644>}
 clavicle_l {parent=spine_05, loc=<5.4831 -0.1053 -3.2946>, rot=<0.0465 -0.7289 -0.009 -0.6829>}
 upperarm_l {parent=clavicle_l, loc=<9.8658 -1.2136 -0.4105>, rot=<0.0164 -0.3245 -0.0269 -0.9454>}
 lowerarm_l {parent=upperarm_l, loc=<26.465 -0.8345 0.259>, rot=<0.0036 0.0161 -0.2298 -0.9731>}
 hand_l {parent=lowerarm_l, loc=<22.7355 -0.3203 -0.6105>, rot=<-0.6784 -0.0223 0.0735 -0.7307>}
 thumb_01_l {parent=hand_l, loc=<1.3118 -0.3191 -2.459>, rot=<0.6546 -0.4042 0.0664 -0.6354>}
 thumb_02_l {parent=thumb_01_l, loc=<5.9336 -0.0448 0.0401>, rot=<-0.0037 0.0213 0.298 -0.9543>}
 thumb_03_l {parent=thumb_02_l, loc=<2.2223 0.3682 -0.0907>, rot=<0.0118 -0.013 0.0066 -0.9998>}
 index_metacarpal_l {parent=hand_l, loc=<1.6535 0.0183 -2.96>, rot=<0.0158 -0.0124 -0.0173 -0.9996>}
 index_01_l {parent=index_metacarpal_l, loc=<8.3008 0.0357 -0.0524>, rot=<-0.0044 0.0485 0.0924 -0.9945>}
 index_02_l {parent=index_01_l, loc=<3.3587 -0.3004 0.0452>, rot=<-0.002 0.0027 0.3225 -0.9466>}
 index_03_l {parent=index_02_l, loc=<2.5578 -0.123 -0.0252>, rot=<-0.002 0.0067 0.2558 -0.9667>}
 middle_metacarpal_l {parent=hand_l, loc=<2.1856 -0.1725 -0.5984>, rot=<0.006 0.0168 -0.0329 -0.9993>}
 middle_01_l {parent=middle_metacarpal_l, loc=<7.3275 -0.2447 0.0129>, rot=<-0.0208 0.0042 0.0858 -0.9961>}
 middle_02_l {parent=middle_01_l, loc=<4.0123 0.1401 0.1175>, rot=<-0.0218 0.0258 0.3512 -0.9357>}
 middle_03_l {parent=middle_02_l, loc=<2.6077 0.0534 0.1373>, rot=<0.0082 0.0188 0.1708 -0.9851>}
 ring_metacarpal_l {parent=hand_l, loc=<2.2074 -0.3625 1.3638>, rot=<-0.0284 0.0722 -0.0458 -0.9959>}
 ring_01_l {parent=ring_metacarpal_l, loc=<7.0295 -0.6197 -0.1446>, rot=<-0.0489 -0.0252 0.1892 -0.9804>}
 ring_02_l {parent=ring_01_l, loc=<3.6174 0.4647 -0.1027>, rot=<0 0 0.2806 -0.9598>}
 ring_03_l {parent=ring_02_l, loc=<2.4152 0.114 0.0199>, rot=<0 0 0.1189 -0.9929>}
 pinky_metacarpal_l {parent=hand_l, loc=<1.7616 -0.9753 3.3814>, rot=<-0.0397 0.0814 -0.069 -0.9935>}
 pinky_01_l {parent=pinky_metacarpal_l, loc=<7.1709 -0.4831 -0.1094>, rot=<-0.0654 -0.0035 0.2242 -0.9723>}
 pinky_02_l {parent=pinky_01_l, loc=<2.8274 0.0893 -0.1043>, rot=<0.0037 0.0011 0.2854 -0.9584>}
 pinky_03_l {parent=pinky_02_l, loc=<2.2046 0.0307 -0.0078>, rot=<-0.0002 0.0098 0.1445 -0.9895>}
 deform_wrist_dn_base_l {parent=hand_l, loc=<0 -1.5326 0>, rot=<0 0 0 -1>}
 deform_wrist_dn_l {parent=deform_wrist_dn_base_l, loc=<0 0 0>, rot=<0 0 0 1>}
 deform_wrist_up_base_l {parent=hand_l, loc=<0 2.3009 0>, rot=<0 0 0 -1>}
 deform_wrist_up_l {parent=deform_wrist_up_base_l, loc=<0 0 0>, rot=<0 0 0 1>}
 lowerarm_twist_02_l {parent=lowerarm_l, loc=<10.6707 0 0>, rot=<0 0 0 1>}
 lowerarm_twist_01_l {parent=lowerarm_l, loc=<18.316 0 0>, rot=<0 0 0 1>}
 deform_elbow_in_base_l {parent=lowerarm_l, loc=<0.1834 3.6062 -0.2572>, rot=<0 0 0 -1>}
 deform_elbow_in_l {parent=deform_elbow_in_base_l, loc=<0 0 0>, rot=<0 0 0 1>}
 deform_elbow_out_base_l {parent=lowerarm_l, loc=<0 -4 0>, rot=<0 0 0 1>}
 deform_elbow_out_l {parent=deform_elbow_out_base_l, loc=<0 0 0>, rot=<0 0 0 1>}
 upperarm_twist_01_l {parent=upperarm_l, loc=<7.2135 -0.1134 0>, rot=<0 0 0 1>}
 deform_upperarm_bk_base_l {parent=upperarm_twist_01_l, loc=<-6.4053 -4.3746 0.314>, rot=<0 0 0 -1>}
 deform_upperarm_bk_l {parent=deform_upperarm_bk_base_l, loc=<0 0 0>, rot=<0 0 0 1>}
 deform_upperarm_fr_base_l {parent=upperarm_twist_01_l, loc=<-5.6312 5.092 0.3354>, rot=<0 0 0 -1>}
 deform_upperarm_fr_l {parent=deform_upperarm_fr_base_l, loc=<0 0 0>, rot=<0 0 0 1>}
 deform_upperarm_up_base_l {parent=upperarm_twist_01_l, loc=<-8.891 0.4825 3.9517>, rot=<0 0 0 -1>}
 deform_upperarm_up_l {parent=deform_upperarm_up_base_l, loc=<0 0 0>, rot=<0 0 0 1>}
 upperarm_twist_02_l {parent=upperarm_l, loc=<19.7998 -0.1134 0>, rot=<0 0 0 1>}
 clavicle_r {parent=spine_05, loc=<5.4831 -0.1053 3.2946>, rot=<0.7289 0.0465 -0.6829 0.009>}
 upperarm_r {parent=clavicle_r, loc=<-9.8658 1.2136 0.4105>, rot=<0.0164 -0.3245 -0.0269 -0.9454>}
 lowerarm_r {parent=upperarm_r, loc=<-26.465 0.8345 -0.259>, rot=<0.0036 0.0161 -0.2298 -0.9731>}
 hand_r {parent=lowerarm_r, loc=<-22.7355 0.3203 0.6106>, rot=<-0.6784 -0.0223 0.0735 -0.7307>}
 thumb_01_r {parent=hand_r, loc=<-1.3118 0.3191 2.459>, rot=<0.6546 -0.4042 0.0664 -0.6354>}
 thumb_02_r {parent=thumb_01_r, loc=<-5.9336 0.0448 -0.0401>, rot=<-0.0037 0.0213 0.298 -0.9543>}
 thumb_03_r {parent=thumb_02_r, loc=<-2.2223 -0.3682 0.0907>, rot=<0.0118 -0.013 0.0066 -0.9998>}
 index_metacarpal_r {parent=hand_r, loc=<-1.6535 -0.0183 2.96>, rot=<0.0158 -0.0124 -0.0173 -0.9996>}
 index_01_r {parent=index_metacarpal_r, loc=<-8.3008 -0.0357 0.0524>, rot=<-0.0044 0.0485 0.0924 -0.9945>}
 index_02_r {parent=index_01_r, loc=<-3.3587 0.3004 -0.0452>, rot=<-0.002 0.0027 0.3225 -0.9466>}
 index_03_r {parent=index_02_r, loc=<-2.5578 0.123 0.0252>, rot=<-0.002 0.0067 0.2558 -0.9667>}
 middle_metacarpal_r {parent=hand_r, loc=<-2.1856 0.1725 0.5984>, rot=<0.006 0.0168 -0.0329 -0.9993>}
 middle_01_r {parent=middle_metacarpal_r, loc=<-7.3275 0.2447 -0.0129>, rot=<-0.0208 0.0042 0.0858 -0.9961>}
 middle_02_r {parent=middle_01_r, loc=<-4.0123 -0.1401 -0.1175>, rot=<-0.0218 0.0258 0.3512 -0.9357>}
 middle_03_r {parent=middle_02_r, loc=<-2.6077 -0.0534 -0.1373>, rot=<0.0082 0.0188 0.1708 -0.9851>}
 ring_metacarpal_r {parent=hand_r, loc=<-2.2074 0.3625 -1.3638>, rot=<-0.0284 0.0722 -0.0458 -0.9959>}
 ring_01_r {parent=ring_metacarpal_r, loc=<-7.0295 0.6197 0.1446>, rot=<-0.0489 -0.0252 0.1892 -0.9804>}
 ring_02_r {parent=ring_01_r, loc=<-3.6174 -0.4647 0.1027>, rot=<0 0 0.2806 -0.9598>}
 ring_03_r {parent=ring_02_r, loc=<-2.4152 -0.114 -0.0199>, rot=<0 0 0.1189 -0.9929>}
 pinky_metacarpal_r {parent=hand_r, loc=<-1.7616 0.9753 -3.3814>, rot=<-0.0397 0.0814 -0.069 -0.9935>}
 pinky_01_r {parent=pinky_metacarpal_r, loc=<-7.1709 0.4831 0.1094>, rot=<-0.0654 -0.0035 0.2242 -0.9723>}
 pinky_02_r {parent=pinky_01_r, loc=<-2.8274 -0.0893 0.1043>, rot=<0.0037 0.0011 0.2854 -0.9584>}
 pinky_03_r {parent=pinky_02_r, loc=<-2.2046 -0.0307 0.0078>, rot=<-0.0002 0.0098 0.1445 -0.9895>}
 deform_wrist_dn_base_r {parent=hand_r, loc=<-0.0001 1.5328 0>, rot=<0 0 0 -1>}
 deform_wrist_dn_r {parent=deform_wrist_dn_base_r, loc=<0 0 0>, rot=<0 0 0 1>}
 deform_wrist_up_base_r {parent=hand_r, loc=<-0.0002 -2.3007 -0.0001>, rot=<0 0 0 -1>}
 deform_wrist_up_r {parent=deform_wrist_up_base_r, loc=<0 0 0>, rot=<0 0 0 1>}
 lowerarm_twist_02_r {parent=lowerarm_r, loc=<-10.6707 0 0>, rot=<0 0 0 -1>}
 lowerarm_twist_01_r {parent=lowerarm_r, loc=<-18.316 0 0>, rot=<0 0 0 -1>}
 deform_elbow_in_base_r {parent=lowerarm_r, loc=<-0.1834 -3.6062 0.2572>, rot=<0 0 0 -1>}
 deform_elbow_in_r {parent=deform_elbow_in_base_r, loc=<0 0 0>, rot=<0 0 0 1>}
 deform_elbow_out_base_r {parent=lowerarm_r, loc=<0 4 -0.0001>, rot=<0 0 0 -1>}
 deform_elbow_out_r {parent=deform_elbow_out_base_r, loc=<0 0 0>, rot=<0 0 0 1>}
 upperarm_twist_01_r {parent=upperarm_r, loc=<-7.2135 0.1134 0>, rot=<0 0 0 -1>}
 deform_upperarm_bk_base_r {parent=upperarm_twist_01_r, loc=<6.4053 4.3746 -0.314>, rot=<0 0 0 -1>}
 deform_upperarm_bk_r {parent=deform_upperarm_bk_base_r, loc=<0 0 0>, rot=<0 0 0 1>}
 deform_upperarm_fr_base_r {parent=upperarm_twist_01_r, loc=<5.6312 -5.092 -0.3354>, rot=<0 0 0 -1>}
 deform_upperarm_fr_r {parent=deform_upperarm_fr_base_r, loc=<0 0 0>, rot=<0 0 0 1>}
 deform_upperarm_up_base_r {parent=upperarm_twist_01_r, loc=<8.891 -0.4825 -3.9517>, rot=<0 0 0 -1>}
 deform_upperarm_up_r {parent=deform_upperarm_up_base_r, loc=<0 0 0>, rot=<0 0 0 1>}
 upperarm_twist_02_r {parent=upperarm_r, loc=<-19.7998 0.1134 0>, rot=<0 0 0 -1>}
 attach_backpack {parent=spine_05, loc=<-0.123 -13.473 0>, rot=<-0.0093 -0.707 0.0093 -0.707>}
 attach_cape {parent=spine_05, loc=<9.017 -11.714 0>, rot=<0.0616 -0.7044 -0.0616 -0.7044>}
 dyn_backstrap_MSTR {parent=spine_05, loc=<1.9317 -9.1446 0>, rot=<0.0644 -0.9979 0 0>}
 dyn_backstrap_01 {parent=dyn_backstrap_MSTR, loc=<0 0 0>, rot=<0 0 0 1>}
 dyn_backstrap_02 {parent=dyn_backstrap_01, loc=<1.7258 0 0>, rot=<0 0 0 1>}
 dyn_zipper_MSTR {parent=spine_05, loc=<-2.3415 11.6808 6.5768>, rot=<-0.2524 -0.9493 -0.1848 0.0304>}
 dyn_zipper_01 {parent=dyn_zipper_MSTR, loc=<0 0 0>, rot=<0 0 0 1>}
 dyn_zipper_02 {parent=dyn_zipper_01, loc=<0.219 0.4619 0.3317>, rot=<0 0 0 1>}
 dyn_hood_MSTR {parent=spine_05, loc=<5.9136 -9.5347 0>, rot=<0.9979 0.0644 0 0>}
 dyn_hood_01 {parent=dyn_hood_MSTR, loc=<0 0 0>, rot=<0 0 0 1>}
 deform_pec_aim_base_l {parent=spine_04, loc=<3.2456 9.3309 -1.2237>, rot=<0.0603 -0.4033 0.3318 -0.8507>}
 deform_pec_aim_l {parent=deform_pec_aim_base_l, loc=<0 0 0>, rot=<0 0 0 1>}
 deform_pec_base_l {parent=deform_pec_aim_l, loc=<9.3665 0 0>, rot=<0 0 0 1>}
 deform_pec_l {parent=deform_pec_base_l, loc=<0 0 0>, rot=<0 0 0 1>}
 deform_pec_aim_base_r {parent=spine_04, loc=<3.2452 9.3309 1.2237>, rot=<-0.4033 -0.0603 0.8507 0.3318>}
 deform_pec_aim_r {parent=deform_pec_aim_base_r, loc=<0 0 0>, rot=<0 0 0 1>}
 deform_pec_base_r {parent=deform_pec_aim_r, loc=<-9.3667 -0.0003 -0.0004>, rot=<0 0 0 1>}
 deform_pec_r {parent=deform_pec_base_r, loc=<0 0 0>, rot=<0 0 0 1>}
 dyn_back_low_jacket_MSTR {parent=spine_04, loc=<0.3234 -8.1848 0>, rot=<0.0416 -0.9991 0 0>}
 dyn_back_low_jacket_01 {parent=dyn_back_low_jacket_MSTR, loc=<0 0 0>, rot=<0 0 0 1>}
 dyn_back_low_jacket_MSTR_l {parent=spine_04, loc=<-0.9298 -4.372 -11.1404>, rot=<0.0045 -0.8594 -0.5113 0.004>}
 dyn_back_low_jacket_01_l {parent=dyn_back_low_jacket_MSTR_l, loc=<0 0 0>, rot=<0 0 0 1>}
 dyn_back_low_jacket_MSTR_r {parent=spine_04, loc=<-0.9294 -4.372 11.1404>, rot=<0.0046 -0.8594 0.5113 -0.004>}
 dyn_back_low_jacket_01_r {parent=dyn_back_low_jacket_MSTR_r, loc=<0 0 0>, rot=<0 0 0 1>}
 dyn_side_jacket_MSTR_l {parent=spine_04, loc=<-3.8836 4.862 -11.1633>, rot=<-0.0239 0.5745 0.8174 -0.034>}
 dyn_side_jacket_01_l {parent=dyn_side_jacket_MSTR_l, loc=<0 0 0>, rot=<0 0 0 1>}
 dyn_side_jacket_MSTR_r {parent=spine_04, loc=<-3.884 4.8619 11.1633>, rot=<0.0239 -0.5745 0.8174 -0.034>}
 dyn_side_jacket_01_r {parent=dyn_side_jacket_MSTR_r, loc=<0 0 0>, rot=<0 0 0 1>}
 thigh_l {parent=pelvis, loc=<-1.0855 -0.2109 -8.7053>, rot=<0.0009 0.0355 0.0155 -0.9993>}
 calf_l {parent=thigh_l, loc=<-42.5284 -0.0019 -0.7297>, rot=<0.0006 -0.0161 -0.0391 -0.9991>}
 foot_l {parent=calf_l, loc=<-46.7401 -2.5014 -0.592>, rot=<0.0427 -0.0238 0.0353 -0.9982>}
 ball_l {parent=foot_l, loc=<-3.5658 10.8066 -0.1544>, rot=<0.0001 -0.0001 -0.7063 -0.7079>}
 calf_twist_02_l {parent=calf_l, loc=<-19.1546 -1.7943 -0.2638>, rot=<0.0237 0.0095 -0.0035 -0.9997>}
 calf_twist_01_l {parent=calf_l, loc=<-35.8483 -1.9197 -0.5794>, rot=<0.0237 0.0095 -0.0035 -0.9997>}
 deform_knee_in_base_l {parent=calf_l, loc=<0.8031 -5.0231 0.2456>, rot=<0 0 0 -1>}
 deform_knee_in_l {parent=deform_knee_in_base_l, loc=<0 0 0>, rot=<0 0 0 1>}
 deform_knee_out_base_l {parent=calf_l, loc=<0 4 0>, rot=<0 0 0 -1>}
 deform_knee_out_l {parent=deform_knee_out_base_l, loc=<0 0 0>, rot=<0 0 0 1>}
 thigh_twist_01_l {parent=thigh_l, loc=<-27.952 0 0>, rot=<0.0013 0 0 -1>}
 dyn_karabin_MSTR {parent=thigh_twist_01_l, loc=<16.4094 -3.3837 -7.9828>, rot=<0.0768 -0.812 -0.5507 0.1771>}
 dyn_karabin_01 {parent=dyn_karabin_MSTR, loc=<0 0 0>, rot=<0 0 0 1>}
 deform_groin_base_l {parent=thigh_l, loc=<0.4801 5.6776 -0.2486>, rot=<0 0 0 -1>}
 deform_groin_l {parent=deform_groin_base_l, loc=<0 0 0>, rot=<0 0 0 1>}
 thigh_r {parent=pelvis, loc=<-1.0855 -0.2109 8.7053>, rot=<0.0355 -0.0009 0.9993 0.0155>}
 calf_r {parent=thigh_r, loc=<42.5284 0.0019 0.7297>, rot=<0.0006 -0.0161 -0.0391 -0.9991>}
 foot_r {parent=calf_r, loc=<46.7401 2.5014 0.592>, rot=<0.0427 -0.0238 0.0353 -0.9982>}
 ball_r {parent=foot_r, loc=<3.5658 -10.8066 0.1544>, rot=<0.0001 -0.0001 -0.7063 -0.7079>}
 calf_twist_02_r {parent=calf_r, loc=<19.1546 1.7943 0.2638>, rot=<0.0237 0.0095 -0.0035 -0.9997>}
 calf_twist_01_r {parent=calf_r, loc=<35.8483 1.9197 0.5794>, rot=<0.0237 0.0095 -0.0035 -0.9997>}
 deform_knee_in_base_r {parent=calf_r, loc=<-0.8031 5.0231 -0.2457>, rot=<0 0 0 -1>}
 deform_knee_in_r {parent=deform_knee_in_base_r, loc=<0 0 0>, rot=<0 0 0 1>}
 deform_knee_out_base_r {parent=calf_r, loc=<0 -4 0>, rot=<0 0 0 -1>}
 deform_knee_out_r {parent=deform_knee_out_base_r, loc=<0 0 0>, rot=<0 0 0 1>}
 thigh_twist_01_r {parent=thigh_r, loc=<27.952 0 0>, rot=<0.0013 0 0 -1>}
 deform_groin_base_r {parent=thigh_r, loc=<-0.4801 -5.6776 0.2486>, rot=<0 0 0 -1>}
 deform_groin_r {parent=deform_groin_base_r, loc=<0 0 0>, rot=<0 0 0 1>}
 deform_glute_base_l {parent=pelvis, loc=<-7.3466 -9.0213 -9.1245>, rot=<0.0172 0.031 -0.4477 -0.8935>}
 deform_glute_l {parent=deform_glute_base_l, loc=<0 0 0>, rot=<0 0 0 1>}
 deform_glute_base_r {parent=pelvis, loc=<-7.3464 -9.0213 9.1245>, rot=<0.031 -0.0172 0.8935 -0.4477>}
 deform_glute_r {parent=deform_glute_base_r, loc=<0 0 0>, rot=<0 0 0 1>}
 extra_can_01 {parent=pelvis, loc=<2.6245 -6.1961 -12.1043>, rot=<-0.0959 -0.9538 0.2335 -0.163>}
 extra_can_02 {parent=pelvis, loc=<-4.573 -5.3679 -14.8625>, rot=<-0.0959 -0.9538 0.2335 -0.163>}
 dyn_belt_MSTR {parent=pelvis, loc=<5.2252 6.8898 8.7524>, rot=<0.0698 -0.5817 0.7928 0.1683>}
 dyn_belt_01 {parent=dyn_belt_MSTR, loc=<0 0 0>, rot=<-0.2636 0 0 -0.9646>}
 dyn_belt_02 {parent=dyn_belt_01, loc=<6.4423 0 0>, rot=<-0.0626 -0.0786 -0.1409 -0.9849>}
 dyn_belt_03 {parent=dyn_belt_02, loc=<4.7256 0 0>, rot=<-0.0704 0.0005 -0.0075 -0.9975>}
 dyn_belt_04 {parent=dyn_belt_03, loc=<6.1726 0 0>, rot=<0 0 -0.0105 -0.9999>}
 dyn_belt_05 {parent=dyn_belt_04, loc=<3.8498 -0.039 0.0001>, rot=<-0.0467 0.0002 0.0054 -0.9989>}
]
