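# Physics asset for the F_MED_VoidSpray female body cosmetic, covering a dangling belt chain plus pelvis and right thigh capsules.
# Capsule bodies run from dyn_belt_MSTR, attached to the pelvis with free rotation, through five linked belt segments.
# Most segments have tuned swing and twist limits, while the last joint locks twist and one swing axis to act like a hinge.
# The right thigh body also joins the pelvis with free rotation.
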PhysicsAsset /BRCosmetics/Characters/Player/Female/Medium/Bodies/F_MED_VoidSpray/Meshes/F_MED_VoidSpray_Physics
bodies {
 dyn_belt_01 [{shapeType=Capsule, length=1.7466, radius=0.5587, shapeName=dyn_belt_01_capsule, center=<3 0 0>}]
 dyn_belt_02 [{shapeType=Capsule, length=2.876, radius=1.0447, shapeName=dyn_belt_02_capsule, center=<2.5 0 0>}]
 dyn_belt_03 [{shapeType=Capsule, length=2.5437, radius=1.0396, shapeName=dyn_belt_03_capsule, center=<3 0 0>}]
 dyn_belt_04 [{shapeType=Capsule, length=4.1873, radius=1.973, shapeName=dyn_belt_04_capsule, center=<4.0915 0 0>, rotation=<rot 90 0 0>}]
 dyn_belt_05 [{shapeType=Capsule, length=2.5624, radius=1.5728, shapeName=dyn_belt_05_capsule, center=<1 0 0>}]
 dyn_belt_MSTR [{shapeType=Capsule, length=0.505, radius=0.505, shapeName=dyn_belt_MSTR_capsule, rotation=<rot 0 -90 0>}]
 pelvis [{shapeType=Capsule, length=15.9333, radius=6.7396, shapeName=pelvis_capsule, rotation=<rot 0 -90 0>}]
 thigh_r [{shapeType=Capsule, length=28.5227, radius=7.7074, shapeName=thigh_r_capsule, center=<16.1887 0 0>, rotation=<rot 90 0 0>}]
}
constraints [
 {
  bone1Name=dyn_belt_05
  bone2Name=dyn_belt_04
  swing1LimitDegrees=90
  swing2LimitDegrees=1
  swing2Motion=Locked
  twistLimitDegrees=1
  twistMotion=Locked
 }
 {
  bone1Name=dyn_belt_MSTR
  bone2Name=pelvis
  swing1LimitDegrees=45
  swing1Motion=Free
  swing2LimitDegrees=45
  swing2Motion=Free
  twistLimitDegrees=45
  twistMotion=Free
 }
 {bone1Name=dyn_belt_01, bone2Name=dyn_belt_MSTR, swing1LimitDegrees=40, swing2LimitDegrees=25, twistLimitDegrees=15}
 {bone1Name=dyn_belt_02, bone2Name=dyn_belt_01, swing1LimitDegrees=40, swing2LimitDegrees=20, twistLimitDegrees=20}
 {bone1Name=dyn_belt_03, bone2Name=dyn_belt_02, swing1LimitDegrees=40, swing2LimitDegrees=20, twistLimitDegrees=20}
 {bone1Name=dyn_belt_04, bone2Name=dyn_belt_03, swing1LimitDegrees=40, swing2LimitDegrees=20, twistLimitDegrees=20}
 {
  bone1Name=thigh_r
  bone2Name=pelvis
  swing1LimitDegrees=45
  swing1Motion=Free
  swing2LimitDegrees=45
  swing2Motion=Free
  twistLimitDegrees=45
  twistMotion=Free
 }
]
