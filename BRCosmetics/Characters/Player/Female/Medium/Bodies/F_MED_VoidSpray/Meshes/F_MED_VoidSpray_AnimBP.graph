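# Anim Blueprint for the F_MED_VoidSpray female cosmetic body, adding secondary motion to pose copied from its attached parent mesh.
# A deform Control Rig feeds a cached pose into pelvis-space RigidBody simulation.
# Chained AnimDynamics nodes then move jacket, zipper, karabiner, hood and backstrap bones before a neck_02 scale tweak.
# Each update sets gravity override, an LOD flag, and front-end (outside emote/skydive) or default sim-space settings.
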AnimBlueprint /BRCosmetics/Characters/Player/Female/Medium/Bodies/F_MED_VoidSpray/Meshes/F_MED_VoidSpray_AnimBP
target_skeleton=&/F_MED_VoidSpray_Skeleton
graphs [
 AnimGraph {nodes [
   n1 = RigidBody!(ComponentPose=Local To Component(LocalPose=Use cached pose 'PreSim'()), Alpha, OverrideWorldGravity:Vector="0,0,0") -> (Pose) {
    bClampLinearTranslationLimitToRefPose=true
    bOverrideWorldGravity=true
    overridePhysicsAsset=&/F_MED_VoidSpray_Physics
    simulationSpace=BaseBoneSpace
    alphaBoolBlend {blendOption=Linear}
    baseBoneRef {boneName=pelvis}
    componentAppliedLinearAccClamp=<3000 3000 3000>
    simSpaceSettings {
     maxAngularAcceleration=20
     maxAngularVelocity=5
     maxLinearAcceleration=1200
     maxLinearVelocity=300
     velocityScaleZ=0.6
     worldAlpha=1
     externalLinearDragV=<1 1 1>
    }
   }
   n2 = Anim Dynamics - Bone: dyn_side_jacket_01_l!(
    ComponentPose=@n1
    Alpha=1
    GravityOverride:Vector="0,0,-980"
    SimSpaceSettings:Anim Phys Sim Space Settings Structure=(SimSpaceAngularAlpha=0,MaxAngularVelocity=10000,MaxAngularAcceleration=10000,ExternalAngularVelocity=(X=0,Y=0,Z=0))
    ExternalForce="0,0,0"
   ) -> (Pose) {
    angularBiasOverride=0.9999
    angularDampingOverride=0.9999
    angularSpringConstant=500
    bOverrideAngularBias=true
    bOverrideAngularDamping=true
    bOverrideLinearDamping=true
    bUseGravityOverride=true
    bUsePlanarLimit=false
    gravityScale=-1
    linearDampingOverride=0.9999
    simulationSpace=RootRelative
    alphaBoolBlend {blendOption=Linear}
    boundBone {boneName=dyn_side_jacket_01_l}
    componentAppliedLinearAccClamp=<5000 5000 5000>
    componentLinearAccScale=<0.2 0.2 0.35>
    componentLinearVelScale=<0.2 0.2 0.35>
    gravityOverride {z=-980}
    physicsBodyDefinitions [{
      collisionType=CustomSphere
      sphereCollisionRadius=1
      boundBone {boneName=dyn_side_jacket_01_l}
      boxExtents=<10 10 14>
      constraintSetup {angularLimitsMax=<2 10 5>, angularLimitsMin=<-2 -10 -90>, angularTarget=<0 0 0>}
      localJointOffset=<4 0 0>
     }]
   }
   n3 = Anim Dynamics - Bone: dyn_side_jacket_01_r!(
    ComponentPose=@n2
    Alpha=1
    GravityOverride:Vector="0,0,-980"
    SimSpaceSettings:Anim Phys Sim Space Settings Structure=(SimSpaceAngularAlpha=0,MaxAngularVelocity=10000,MaxAngularAcceleration=10000,ExternalAngularVelocity=(X=0,Y=0,Z=0))
    ExternalForce="0,0,0"
   ) -> (Pose) {
    angularBiasOverride=0.9999
    angularDampingOverride=0.9999
    angularSpringConstant=500
    bOverrideAngularBias=true
    bOverrideAngularDamping=true
    bOverrideLinearDamping=true
    bUseGravityOverride=true
    bUsePlanarLimit=false
    gravityScale=-1
    linearDampingOverride=0.9999
    simulationSpace=RootRelative
    alphaBoolBlend {blendOption=Linear}
    boundBone {boneName=dyn_side_jacket_01_r}
    componentAppliedLinearAccClamp=<5000 5000 5000>
    componentLinearAccScale=<0.2 0.2 0.35>
    componentLinearVelScale=<0.2 0.2 0.35>
    gravityOverride {z=-980}
    physicsBodyDefinitions [{
      collisionType=CustomSphere
      sphereCollisionRadius=1
      boundBone {boneName=dyn_side_jacket_01_r}
      boxExtents=<10 10 14>
      constraintSetup {angularLimitsMax=<2 10 5>, angularLimitsMin=<-2 -10 -90>, angularTarget=<0 0 0>}
      localJointOffset=<4 0 0>
     }]
   }
   n4 = Anim Dynamics - Bone: dyn_back_low_jacket_01_l!(
    ComponentPose=@n3
    Alpha=1
    GravityOverride:Vector="0,0,-980"
    SimSpaceSettings:Anim Phys Sim Space Settings Structure=(SimSpaceAngularAlpha=0,MaxAngularVelocity=10000,MaxAngularAcceleration=10000,ExternalAngularVelocity=(X=0,Y=0,Z=0))
    ExternalForce="0,0,0"
   ) -> (Pose) {
    angularBiasOverride=0.9999
    angularDampingOverride=0.9999
    angularSpringConstant=500
    bOverrideAngularBias=true
    bOverrideAngularDamping=true
    bOverrideLinearDamping=true
    bUseGravityOverride=true
    bUsePlanarLimit=false
    gravityScale=-1
    linearDampingOverride=0.9999
    simulationSpace=RootRelative
    alphaBoolBlend {blendOption=Linear}
    boundBone {boneName=dyn_back_low_jacket_01_l}
    componentAppliedLinearAccClamp=<5000 5000 5000>
    componentLinearAccScale=<0.4 0.4 0.3>
    componentLinearVelScale=<0.4 0.4 0.3>
    gravityOverride {z=-980}
    physicsBodyDefinitions [{
      collisionType=CustomSphere
      sphereCollisionRadius=1
      boundBone {boneName=dyn_back_low_jacket_01_l}
      boxExtents=<5 5 5>
      constraintSetup {angularLimitsMax=<4 4 5>, angularLimitsMin=<-4 -4 -90>, angularTarget=<0 0 0>}
      localJointOffset=<2 0 0>
     }]
   }
   n5 = Anim Dynamics - Bone: dyn_back_low_jacket_01_r!(
    ComponentPose=@n4
    Alpha=1
    GravityOverride:Vector="0,0,-980"
    SimSpaceSettings:Anim Phys Sim Space Settings Structure=(SimSpaceAngularAlpha=0,MaxAngularVelocity=10000,MaxAngularAcceleration=10000,ExternalAngularVelocity=(X=0,Y=0,Z=0))
    ExternalForce="0,0,0"
   ) -> (Pose) {
    angularBiasOverride=0.9999
    angularDampingOverride=0.9999
    angularSpringConstant=500
    bOverrideAngularBias=true
    bOverrideAngularDamping=true
    bOverrideLinearDamping=true
    bUseGravityOverride=true
    bUsePlanarLimit=false
    gravityScale=-1
    linearDampingOverride=0.9999
    simulationSpace=RootRelative
    alphaBoolBlend {blendOption=Linear}
    boundBone {boneName=dyn_back_low_jacket_01_r}
    componentAppliedLinearAccClamp=<5000 5000 5000>
    componentLinearAccScale=<0.4 0.4 0.3>
    componentLinearVelScale=<0.4 0.4 0.3>
    gravityOverride {z=-980}
    physicsBodyDefinitions [{
      collisionType=CustomSphere
      sphereCollisionRadius=1
      boundBone {boneName=dyn_back_low_jacket_01_r}
      boxExtents=<5 5 5>
      constraintSetup {angularLimitsMax=<4 4 5>, angularLimitsMin=<-4 -4 -90>, angularTarget=<0 0 0>}
      localJointOffset=<2 0 0>
     }]
   }
   n6 = Anim Dynamics - Bone: dyn_back_low_jacket_01!(
    ComponentPose=@n5
    Alpha=1
    GravityOverride:Vector="0,0,-980"
    SimSpaceSettings:Anim Phys Sim Space Settings Structure=(SimSpaceAngularAlpha=0,MaxAngularVelocity=10000,MaxAngularAcceleration=10000,ExternalAngularVelocity=(X=0,Y=0,Z=0))
    ExternalForce="0,0,0"
   ) -> (Pose) {
    angularBiasOverride=0.9999
    angularDampingOverride=0.9999
    angularSpringConstant=500
    bOverrideAngularBias=true
    bOverrideAngularDamping=true
    bOverrideLinearDamping=true
    bUseGravityOverride=true
    bUsePlanarLimit=false
    gravityScale=-1
    linearDampingOverride=0.9999
    simulationSpace=RootRelative
    alphaBoolBlend {blendOption=Linear}
    boundBone {boneName=dyn_back_low_jacket_01}
    componentAppliedLinearAccClamp=<5000 5000 5000>
    componentLinearAccScale=<0.4 0.4 0.3>
    componentLinearVelScale=<0.4 0.4 0.3>
    gravityOverride {z=-980}
    physicsBodyDefinitions [{
      collisionType=CustomSphere
      sphereCollisionRadius=1
      boundBone {boneName=dyn_back_low_jacket_01}
      boxExtents=<10 10 10>
      constraintSetup {angularLimitsMax=<4 4 5>, angularLimitsMin=<-4 -2 -90>, angularTarget=<0 0 0>}
      localJointOffset=<4 0 0>
     }]
   }
   n7 = Anim Dynamics - Bone: dyn_zipper_02!(
    ComponentPose=@n6
    Alpha=1
    GravityOverride:Vector="0,0,-980"
    SimSpaceSettings:Anim Phys Sim Space Settings Structure=(SimSpaceAngularAlpha=0,MaxAngularVelocity=10000,MaxAngularAcceleration=10000,ExternalAngularVelocity=(X=0,Y=0,Z=0))
    ExternalForce="0,0,0"
   ) -> (Pose) {
    angularBiasOverride=0.9999
    angularDampingOverride=0.9999
    angularSpringConstant=500
    bOverrideAngularBias=true
    bOverrideAngularDamping=true
    bOverrideLinearDamping=true
    bUseGravityOverride=true
    bUsePlanarLimit=false
    gravityScale=-1
    linearDampingOverride=0.9999
    simulationSpace=RootRelative
    alphaBoolBlend {blendOption=Linear}
    boundBone {boneName=dyn_zipper_02}
    componentAppliedLinearAccClamp=<5000 5000 5000>
    componentLinearAccScale=<0.4 0.4 0.4>
    componentLinearVelScale=<0.4 0.4 0.4>
    gravityOverride {z=-980}
    physicsBodyDefinitions [{
      collisionType=CustomSphere
      sphereCollisionRadius=1
      boundBone {boneName=dyn_zipper_02}
      boxExtents=<10 5 5>
      constraintSetup {angularLimitsMax=<2 15 130>, angularLimitsMin=<-2 -15 0>, angularTarget=<0 0 0>}
      localJointOffset=<3 0 0>
     }]
   }
   n8 = Anim Dynamics - Bone: dyn_karabin_01!(
    ComponentPose=@n7
    Alpha=1
    GravityOverride:Vector="0,0,-980"
    SimSpaceSettings:Anim Phys Sim Space Settings Structure=(SimSpaceAngularAlpha=0,MaxAngularVelocity=10000,MaxAngularAcceleration=10000,ExternalAngularVelocity=(X=0,Y=0,Z=0))
    ExternalForce="0,0,0"
   ) -> (Pose) {
    angularBiasOverride=0.9999
    angularDampingOverride=0.9999
    angularSpringConstant=500
    bOverrideAngularBias=true
    bOverrideAngularDamping=true
    bOverrideLinearDamping=true
    bUseGravityOverride=true
    bUsePlanarLimit=false
    gravityScale=-1
    linearDampingOverride=0.9999
    simulationSpace=RootRelative
    alphaBoolBlend {blendOption=Linear}
    boundBone {boneName=dyn_karabin_01}
    componentAppliedLinearAccClamp=<5000 5000 5000>
    componentLinearAccScale=<0.3 0.3 0.4>
    componentLinearVelScale=<0.3 0.3 0.4>
    gravityOverride {z=-980}
    physicsBodyDefinitions [{
      collisionType=CustomSphere
      sphereCollisionRadius=1
      boundBone {boneName=dyn_karabin_01}
      boxExtents=<10 10 10>
      constraintSetup {angularLimitsMax=<1 1 0>, angularLimitsMin=<-1 -1 -100>, angularTarget=<0 0 0>}
      localJointOffset=<4 0 0>
     }]
   }
   n9 = Anim Dynamics - Bone: dyn_hood_01!(
    ComponentPose=@n8
    Alpha=1
    GravityOverride:Vector="0,0,-980"
    SimSpaceSettings:Anim Phys Sim Space Settings Structure=(SimSpaceAngularAlpha=0,MaxAngularVelocity=10000,MaxAngularAcceleration=10000,ExternalAngularVelocity=(X=0,Y=0,Z=0))
    ExternalForce="0,0,0"
   ) -> (Pose) {
    angularBiasOverride=0.9999
    angularDampingOverride=0.9999
    angularSpringConstant=5000
    bAngularSpring=true
    bOverrideAngularBias=true
    bOverrideAngularDamping=true
    bOverrideLinearDamping=true
    bUseGravityOverride=true
    bUsePlanarLimit=false
    gravityScale=-1
    linearDampingOverride=0.9999
    simulationSpace=RootRelative
    alphaBoolBlend {blendOption=Linear}
    boundBone {boneName=dyn_hood_01}
    componentAppliedLinearAccClamp=<5000 5000 5000>
    componentLinearAccScale=<0.3 0.3 0.4>
    componentLinearVelScale=<0.3 0.3 0.4>
    gravityOverride {z=-980}
    physicsBodyDefinitions [{
      collisionType=CustomSphere
      sphereCollisionRadius=1
      boundBone {boneName=dyn_hood_01}
      boxExtents=<8 8 16>
      constraintSetup {angularLimitsMax=<3 2 3>, angularLimitsMin=<-3 -2 -10>, angularTarget=<0 0 1>}
      localJointOffset=<4 0 0>
     }]
   }
   n10 = Anim Dynamics - Bone: dyn_backstrap_01!(
    ComponentPose=@n9
    Alpha=1
    GravityOverride:Vector="0,0,-980"
    SimSpaceSettings:Anim Phys Sim Space Settings Structure=(SimSpaceAngularAlpha=0,MaxAngularVelocity=10000,MaxAngularAcceleration=10000,ExternalAngularVelocity=(X=0,Y=0,Z=0))
    ExternalForce="0,0,0"
   ) -> (Pose) {
    angularBiasOverride=0.9999
    angularDampingOverride=0.9999
    angularSpringConstant=500
    bOverrideAngularBias=true
    bOverrideAngularDamping=true
    bOverrideLinearDamping=true
    bUseGravityOverride=true
    bUsePlanarLimit=false
    gravityScale=-1
    linearDampingOverride=0.9999
    simulationSpace=RootRelative
    worldSpaceMinimumScale=0.001
    alphaBoolBlend {blendOption=Linear}
    boundBone {boneName=dyn_backstrap_01}
    componentAppliedLinearAccClamp=<5000 5000 5000>
    componentLinearAccScale=<0.5 0.5 0.5>
    componentLinearVelScale=<0.5 0.5 0.5>
    gravityOverride {z=-980}
    physicsBodyDefinitions [{
      collisionType=CustomSphere
      sphereCollisionRadius=1
      boundBone {boneName=dyn_backstrap_01}
      boxExtents=<5 5 5>
      constraintSetup {angularLimitsMax=<1 5 1>, angularLimitsMin=<-1 -5 -10>, angularTarget=<0 0 0>}
      localJointOffset=<2 0 0>
     }]
   }
   n11 = Anim Dynamics - Bone: dyn_backstrap_02!(
    ComponentPose=@n10
    Alpha=1
    GravityOverride:Vector="0,0,-980"
    SimSpaceSettings:Anim Phys Sim Space Settings Structure=(SimSpaceAngularAlpha=0,MaxAngularVelocity=10000,MaxAngularAcceleration=10000,ExternalAngularVelocity=(X=0,Y=0,Z=0))
    ExternalForce="0,0,0"
   ) -> (Pose) {
    angularBiasOverride=0.9999
    angularDampingOverride=0.9999
    angularSpringConstant=500
    bOverrideAngularBias=true
    bOverrideAngularDamping=true
    bOverrideLinearDamping=true
    bUseGravityOverride=true
    bUsePlanarLimit=false
    gravityScale=-1
    linearDampingOverride=0.9999
    simulationSpace=RootRelative
    alphaBoolBlend {blendOption=Linear}
    boundBone {boneName=dyn_backstrap_02}
    componentAppliedLinearAccClamp=<5000 5000 5000>
    componentLinearAccScale=<0.5 0.5 0.5>
    componentLinearVelScale=<0.5 0.5 0.5>
    gravityOverride {z=-980}
    physicsBodyDefinitions [{
      collisionType=CustomSphere
      sphereCollisionRadius=1
      boundBone {boneName=dyn_backstrap_02}
      boxExtents=<5 5 5>
      constraintSetup {angularLimitsMax=<0 0 1>, angularLimitsMin=<0 0 -110>, angularTarget=<0 0 0>}
      localJointOffset=<1 0 0>
     }]
   }
   n12 = "Transform (Modify) Bone - Bone: neck_02"!(ComponentPose=@n11, bAlphaBoolEnabled:bool=True, Translation="0,0,0", Rotation="0,0,0", Scale="1,1.15,1.15") -> (Pose) {alphaInputType=Bool, scaleMode=BMM_Additive, scaleSpace=BCS_BoneSpace, boneToModify {boneName=neck_02}, scale {y=1.15, z=1.15}}
   Output Pose(Result=Component To Local(ComponentPose=@n12)) {layerGroup=None}
   n13 = Control Rig(Source=Copy Pose From Mesh() {bUseAttachedParent=true}) {controlRigAssetReference {blueprintRigClass=&/F_MED_VoidSpray_CtrlRig_Deform.F_MED_VoidSpray_CtrlRig_Deform_C}}
   Save cached pose 'PreSim'(Pose=@n13) {cacheName=PreSim}
  ]}
 EventGraph {nodes [
   Event BlueprintUpdateAnimation()
   > n1 = Sequence()
   > n2 = "Gravity Override State Function (Parameter Asset) (Nativized)"(
    CharacterPartAnimInstance=Self-Reference()
    GameTimeInSeconds=GetGameTimeInSeconds()
    Parameters=&/F_MED_VoidSpray_Parameters
   )
   > Set GravityOverride(GravityOverride=@n2)
   > Set LOD1(LOD1:bool="integer > integer"(A=Get CurrentLOD())) -> (Output_Get:bool)
   n3 = NOT Boolean(A=Get bIsPlayingEmote())
   n4 = NOT Boolean(A=Get bIsSkyDiving())
   n5 = Branch(execute=@n1.then_1, Condition=AND Boolean(A=Get bIsInFrontEnd(), B=AND Boolean(A=@n4, B=@n3)))
   > Set SimSpaceSettings(SimSpaceSettings:Anim Phys Sim Space Settings Structure=Get SimSpaceSettingsFE() -> (SimSpaceSettingsFE:Anim Phys Sim Space Settings Structure)) -> (Output_Get:Anim Phys Sim Space Settings Structure)
   Set SimSpaceSettings(execute=@n5.else, SimSpaceSettings=Get SimSpaceSettingsPIE() -> (SimSpaceSettingsPIE:Anim Phys Sim Space Settings Structure))
  ]}
]
variables [
 GravityOverride {type=Vector}
 LOD1 {type=boolean}
 SimSpaceSettings {
  type=AnimPhysSimSpaceSettings
  default {maxAngularAcceleration=100000, maxAngularVelocity=100000, simSpaceAngularAlpha=1, externalAngularVelocity=<100000 100000 100000>}
 }
 SimSpaceSettingsFE {
  type=AnimPhysSimSpaceSettings
  default {maxAngularAcceleration=50, maxAngularVelocity=3, simSpaceAngularAlpha=1, externalAngularVelocity=<0 0 0>}
 }
 SimSpaceSettingsPIE {
  type=AnimPhysSimSpaceSettings
  default {maxAngularAcceleration=10, maxAngularVelocity=10, simSpaceAngularAlpha=0, externalAngularVelocity=<0 0 2>}
 }
]
