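# Physics asset for the face accessory part of the female medium VoidSpray character, covering a hair tail and left and right side locks.
# Capsule bodies form a five-bone tail and two-bone side locks, each chained by angle-limited constraints to a small MSTR root body.
# The tail's twist loosens toward the tip, while the side locks are tighter.
# Head, neck, upper spine and backpack capsules supply collision shapes, and a free-angle constraint links head to first tail bone.
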PhysicsAsset /BRCosmetics/Characters/Player/Female/Medium/Bodies/F_MED_VoidSpray/Meshes/Parts/F_MED_VoidSpray_FaceAcc_Physics
bodies {
 dyn_dyn_hairSide_l_01 [{shapeType=Capsule, length=3.4379, radius=1.5759, shapeName=dyn_dyn_hairSide_l_01_capsule, center=<1 0 0>, rotation=<rot 0 0 90>}]
 dyn_dyn_hairSide_l_02 [{shapeType=Capsule, length=5.2129, radius=2.5498, shapeName=dyn_dyn_hairSide_l_02_capsule, center=<3.6212 0 0>, rotation=<rot 90 0 0>}]
 dyn_dyn_hairSide_r_01 [{shapeType=Capsule, length=2.9859, radius=0.8271, shapeName=dyn_dyn_hairSide_r_01_capsule, center=<1.5706 0 0>, rotation=<rot 0 0 90>}]
 dyn_dyn_hairSide_r_02 [{shapeType=Capsule, length=7.7712, radius=1.7493, shapeName=dyn_dyn_hairSide_r_02_capsule, center=<4.2758 0.6448 0>, rotation=<rot 90 0 0>}]
 dyn_dyn_hairTail_01 [{shapeType=Capsule, length=0.3895, radius=4.1128, shapeName=dyn_dyn_hairTail_01_capsule, center=<3 0 0>, rotation=<rot 90 0 0>}]
 dyn_dyn_hairTail_02 [{shapeType=Capsule, length=5.6082, radius=3.7584, shapeName=dyn_dyn_hairTail_02_capsule, center=<4.7195 0 0>, rotation=<rot 90 0 0>}]
 dyn_dyn_hairTail_03 [{shapeType=Capsule, length=3.887, radius=3.2878, shapeName=dyn_dyn_hairTail_03_capsule, center=<5 0 0>, rotation=<rot 90 0 0>}]
 dyn_dyn_hairTail_04 [{shapeType=Capsule, length=4.5255, radius=2.8022, shapeName=dyn_dyn_hairTail_04_capsule, center=<2.396 0.0675 0.0039>, rotation=<rot 90 0 0>}]
 dyn_dyn_hairTail_05 [{
   shapeType=Capsule
   length=2.8211
   radius=1.6204
   shapeName=dyn_dyn_hairTail_05_capsule
   center=<2.3325 0 0>
   rotation=<rot -0.4079 -97.0276 -87.6859>
  }]
 dyn_hairSide_l_MSTR [{shapeType=Capsule, length=0.505, radius=0.505, shapeName=dyn_hairSide_l_MSTR_capsule, rotation=<rot 0 -90 0>}]
 dyn_hairSide_r_MSTR [{shapeType=Capsule, length=0.505, radius=0.505, shapeName=dyn_hairSide_r_MSTR_capsule, rotation=<rot 0 -90 0>}]
 dyn_hairTail_MSTR [{shapeType=Capsule, length=0.505, radius=0.505, shapeName=dyn_hairTail_MSTR_capsule, rotation=<rot 0 -90 0>}]
 extra_backpack_collider [{shapeType=Capsule, length=15, radius=10, shapeName=extra_backpack_collider_capsule, rotation=<rot 0 -90 0>}]
 head [{shapeType=Capsule, length=4.9949, radius=6.3396, shapeName=head_capsule, center=<3.8411 -1.1539 0>, rotation=<rot 90 0 0>}]
 neck_01 [{
   shapeType=Capsule
   length=3.3674
   radius=3.5878
   shapeName=neck_01_capsule
   center=<4.1081 1.6193 -0.6485>
   rotation=<rot 1.5708 -110.3743 -87.7842>
  }]
 spine_05 [
  {shapeType=Capsule, length=17.603, radius=11.6451, shapeName=spine_05_capsule, rotation=<rot 0 -90 0>}
  {shapeType=Capsule, length=19.0126, radius=6.0155, shapeName=spine_05_capsule, center=<9.4351 -5.4374 0>}
 ]
}
constraints [
 {bone1Name=dyn_dyn_hairTail_05, bone2Name=dyn_dyn_hairTail_04, swing1LimitDegrees=45, swing2LimitDegrees=45, twistLimitDegrees=25}
 {bone1Name=dyn_dyn_hairTail_01, bone2Name=dyn_hairTail_MSTR, swing1LimitDegrees=25, swing2LimitDegrees=25, twistLimitDegrees=10}
 {bone1Name=dyn_dyn_hairTail_02, bone2Name=dyn_dyn_hairTail_01, swing1LimitDegrees=45, swing2LimitDegrees=45, twistLimitDegrees=10}
 {bone1Name=dyn_dyn_hairTail_03, bone2Name=dyn_dyn_hairTail_02, swing1LimitDegrees=45, swing2LimitDegrees=45, twistLimitDegrees=15}
 {bone1Name=dyn_dyn_hairTail_04, bone2Name=dyn_dyn_hairTail_03, swing1LimitDegrees=45, swing2LimitDegrees=45, twistLimitDegrees=20}
 {bone1Name=dyn_dyn_hairSide_r_02, bone2Name=dyn_dyn_hairSide_r_01, swing1LimitDegrees=25, swing2LimitDegrees=15, twistLimitDegrees=10}
 {bone1Name=dyn_dyn_hairSide_r_01, bone2Name=dyn_hairSide_r_MSTR, swing1LimitDegrees=15, swing2LimitDegrees=5, twistLimitDegrees=5}
 {bone1Name=dyn_dyn_hairSide_l_02, bone2Name=dyn_dyn_hairSide_l_01, swing1LimitDegrees=25, swing2LimitDegrees=15, twistLimitDegrees=10}
 {bone1Name=dyn_dyn_hairSide_l_01, bone2Name=dyn_hairSide_l_MSTR, swing1LimitDegrees=15, swing2LimitDegrees=5, twistLimitDegrees=5}
 {
  bone1Name=head
  bone2Name=dyn_dyn_hairTail_01
  swing1LimitDegrees=45
  swing1Motion=Free
  swing2LimitDegrees=45
  swing2Motion=Free
  twistLimitDegrees=45
  twistMotion=Free
 }
]
