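# Post-process style AnimBP for the VoidSpray female face accessory that adds secondary physics motion, mainly to hair bones.
# It copies the attached parent's pose, offsets the extra backpack collider, then runs RigidBody simulation in neck_01 space and four spring-driven AnimDynamics hair bones.
# The event graph computes a gravity override from a parameter asset, selects front-end or in-game sim-space settings, and flags backpacks or capes when customization loads.
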AnimBlueprint /BRCosmetics/Characters/Player/Female/Medium/Bodies/F_MED_VoidSpray/Meshes/Parts/F_MED_VoidSpray_FaceAcc_AnimBP
target_skeleton=&/F_MED_VoidSpray_FaceAcc_Skeleton
graphs [
 AnimGraph {nodes [
   n1 = "Transform (Modify) Bone - Bone: extra_backpack_collider"(
    ComponentPose=Local To Component(LocalPose=Copy Pose From Mesh() {bUseAttachedParent=true})
    Alpha=1
    Translation="-0.000,-14.959,7.762"
    Rotation="0,0,0"
    Scale="1,1,1"
   ) {
    rotationMode=BMM_Additive
    translationMode=BMM_Additive
    alphaScaleBiasClamp {bInterpResult=true, interpSpeedDecreasing=1, interpSpeedIncreasing=1}
    boneToModify {boneName=extra_backpack_collider}
    translation {y=-14.959, z=7.762}
   }
   n2 = RigidBody!(ComponentPose=@n1, Alpha, OverrideWorldGravity:Vector="0,0,-1680") -> (Pose) {
    bClampLinearTranslationLimitToRefPose=true
    bOverrideWorldGravity=true
    overridePhysicsAsset=&/F_MED_VoidSpray_FaceAcc_Physics
    simulationSpace=BaseBoneSpace
    alphaBoolBlend {blendOption=Linear}
    alphaScaleBiasClamp {bInterpResult=true, interpSpeedIncreasing=2}
    baseBoneRef {boneName=neck_01}
    componentAppliedLinearAccClamp=<1000 1000 1000>
    simSpaceSettings {
     maxAngularAcceleration=32
     maxAngularVelocity=4
     maxLinearAcceleration=1150
     maxLinearVelocity=500
     worldAlpha=0.8
     externalLinearDragV=<2 2 1>
    }
   }
   n3 = Anim Dynamics - Bone: dyn_dyn_hair_f_m_01!(
    ComponentPose=@n2
    Alpha=1
    GravityOverride:Vector="0,0,0"
    SimSpaceSettings:Anim Phys Sim Space Settings Structure=(SimSpaceAngularAlpha=0,MaxAngularVelocity=10000,MaxAngularAcceleration=10000,ExternalAngularVelocity=(X=0,Y=0,Z=0))
    ExternalForce="0,0,0"
   ) -> (Pose) {
    angularBiasOverride=0.9999
    angularDampingOverride=0.9999
    bLinearSpring=true
    bOverrideAngularBias=true
    bOverrideAngularDamping=true
    bOverrideLinearDamping=true
    bUseGravityOverride=true
    linearDampingOverride=0.9999
    linearSpringConstant=100
    simulationSpace=RootRelative
    alphaBoolBlend {blendOption=Linear}
    alphaScaleBiasClamp {scale=0.55}
    boundBone {boneName=dyn_dyn_hair_f_m_01}
    componentAppliedLinearAccClamp=<1250 1250 1250>
    componentLinearAccScale=<0.4 0.5 0.5>
    componentLinearVelScale=<0.4 0.5 0.5>
    physicsBodyDefinitions [{
      boundBone {boneName=dyn_dyn_hair_f_m_01}
      boxExtents=<8 8 12>
      constraintSetup {angularTargetAxis=AxisZ, angularLimitsMax=<5 15 15>, angularLimitsMin=<-5 -15 0>, angularTarget=<1 1 1>}
      localJointOffset=<3 0 0>
     }]
   }
   n4 = Anim Dynamics - Bone: dyn_dyn_hair_f_m_02!(
    ComponentPose=@n3
    Alpha=1
    GravityOverride:Vector="0,0,0"
    SimSpaceSettings:Anim Phys Sim Space Settings Structure=(SimSpaceAngularAlpha=0,MaxAngularVelocity=10000,MaxAngularAcceleration=10000,ExternalAngularVelocity=(X=0,Y=0,Z=0))
    ExternalForce="0,0,0"
   ) -> (Pose) {
    angularBiasOverride=0.9999
    angularDampingOverride=0.9999
    bLinearSpring=true
    bOverrideAngularBias=true
    bOverrideAngularDamping=true
    bOverrideLinearDamping=true
    bUseGravityOverride=true
    linearDampingOverride=0.9999
    linearSpringConstant=100
    simulationSpace=RootRelative
    alphaBoolBlend {blendOption=Linear}
    alphaScaleBiasClamp {scale=0.55}
    boundBone {boneName=dyn_dyn_hair_f_m_02}
    componentAppliedLinearAccClamp=<1250 1250 1250>
    componentLinearAccScale=<0.4 0.5 0.5>
    componentLinearVelScale=<0.4 0.5 0.5>
    physicsBodyDefinitions [{
      boundBone {boneName=dyn_dyn_hair_f_m_02}
      boxExtents=<8 8 12>
      constraintSetup {angularTargetAxis=AxisZ, angularLimitsMax=<5 15 15>, angularLimitsMin=<-5 -15 0>, angularTarget=<1 1 1>}
      localJointOffset=<3 0 0>
     }]
   }
   n5 = Anim Dynamics - Bone: dyn_dyn_hair_f_l_01!(
    ComponentPose=@n4
    Alpha=1
    GravityOverride:Vector="0,0,0"
    SimSpaceSettings:Anim Phys Sim Space Settings Structure=(SimSpaceAngularAlpha=0,MaxAngularVelocity=10000,MaxAngularAcceleration=10000,ExternalAngularVelocity=(X=0,Y=0,Z=0))
    ExternalForce="0,0,0"
   ) -> (Pose) {
    angularBiasOverride=0.9999
    angularDampingOverride=0.9999
    bLinearSpring=true
    bOverrideAngularBias=true
    bOverrideAngularDamping=true
    bOverrideLinearDamping=true
    bUseGravityOverride=true
    linearDampingOverride=0.9999
    linearSpringConstant=100
    simulationSpace=RootRelative
    alphaBoolBlend {blendOption=Linear}
    alphaScaleBiasClamp {scale=0.45}
    boundBone {boneName=dyn_dyn_hair_f_l_01}
    componentAppliedLinearAccClamp=<1250 1250 1250>
    componentLinearAccScale=<0.4 0.5 0.5>
    componentLinearVelScale=<0.4 0.5 0.5>
    physicsBodyDefinitions [{
      boundBone {boneName=dyn_dyn_hair_f_l_01}
      boxExtents=<8 8 12>
      constraintSetup {angularTargetAxis=AxisZ, angularLimitsMax=<5 15 15>, angularLimitsMin=<-5 -15 0>, angularTarget=<1 1 1>}
      localJointOffset=<3 0 0>
     }]
   }
   n6 = Anim Dynamics - Bone: dyn_dyn_hair_f_l_02!(
    ComponentPose=@n5
    Alpha=1
    GravityOverride:Vector="0,0,0"
    SimSpaceSettings:Anim Phys Sim Space Settings Structure=(SimSpaceAngularAlpha=0,MaxAngularVelocity=10000,MaxAngularAcceleration=10000,ExternalAngularVelocity=(X=0,Y=0,Z=0))
    ExternalForce="0,0,0"
   ) -> (Pose) {
    angularBiasOverride=0.9999
    angularDampingOverride=0.9999
    bLinearSpring=true
    bOverrideAngularBias=true
    bOverrideAngularDamping=true
    bOverrideLinearDamping=true
    bUseGravityOverride=true
    linearDampingOverride=0.9999
    linearSpringConstant=100
    simulationSpace=RootRelative
    alphaBoolBlend {blendOption=Linear}
    alphaScaleBiasClamp {scale=0.45}
    boundBone {boneName=dyn_dyn_hair_f_l_02}
    componentAppliedLinearAccClamp=<1250 1250 1250>
    componentLinearAccScale=<0.4 0.5 0.5>
    componentLinearVelScale=<0.4 0.5 0.5>
    physicsBodyDefinitions [{
      boundBone {boneName=dyn_dyn_hair_f_l_02}
      boxExtents=<8 8 12>
      constraintSetup {angularTargetAxis=AxisZ, angularLimitsMax=<5 15 15>, angularLimitsMin=<-5 -15 0>, angularTarget=<1 1 1>}
      localJointOffset=<3 0 0>
     }]
   }
   Output Pose(Result=Component To Local(ComponentPose=@n6)) {layerGroup=None}
  ]}
 EventGraph {nodes [
   Event BlueprintUpdateAnimation()
   > n1 = "Gravity Override State Function (Parameter Asset) (Nativized)"(
    CharacterPartAnimInstance=Self-Reference()
    GameTimeInSeconds=GetGameTimeInSeconds()
    Parameters=/BRCosmetics/Characters/Player/Female/Medium/Bodies/F_MED_VoidSpray/Meshes/F_MED_VoidSpray_Parameters
   )
   > Set GravityOverride(GravityOverride=@n1)
   n2 = NOT Boolean(A=Get bIsPlayingEmote())
   n3 = NOT Boolean(A=Get bIsSkyDiving())
   > n4 = Branch(Condition=AND Boolean(A=Get bIsInFrontEnd(), B=AND Boolean(A=@n3, B=@n2)))
   > Set SimSpaceSettings(SimSpaceSettings:Anim Phys Sim Space Settings Structure=Get SimSpaceSettingsFE() -> (SimSpaceSettingsFE:Anim Phys Sim Space Settings Structure)) -> (Output_Get:Anim Phys Sim Space Settings Structure)
   Set SimSpaceSettings(execute=@n4.else, SimSpaceSettings=Get SimSpaceSettingsPIE() -> (SimSpaceSettingsPIE:Anim Phys Sim Space Settings Structure))
   Event OnCharacterCustomizationLoaded()
   n5 = CheckForBackbling()
   > Set IsWearingBackpackOrCape(IsWearingBackpackOrCape:bool=OR Boolean!(A=@n5, B=@n5.bIsBackblingACape, C:bool=@n5.bIsBackblingLong) -> (ReturnValue)) -> (Output_Get:bool)
   > n6 = Branch(Condition=Get IsWearingBackpackOrCape() -> (IsWearingBackpackOrCape:bool))
   > Set IsWearingBackpack(IsWearingBackpack:double=1) -> (Output_Get:double)
   Set IsWearingBackpack(execute=@n6.else, IsWearingBackpack=0)
  ]}
]
variables [
 GravityOverride {type=Vector}
 IsWearingBackpack {type=number}
 IsWearingBackpackOrCape {type=boolean}
 SimSpaceSettings {
  type=AnimPhysSimSpaceSettings
  default {maxAngularAcceleration=100000, maxAngularVelocity=100000, simSpaceAngularAlpha=1, externalAngularVelocity=<100000 100000 100000>}
 }
 SimSpaceSettingsFE {
  type=AnimPhysSimSpaceSettings
  default {maxAngularAcceleration=75, maxAngularVelocity=5, simSpaceAngularAlpha=0.7, externalAngularVelocity=<1 1 1>}
 }
 SimSpaceSettingsPIE {
  type=AnimPhysSimSpaceSettings
  default {maxAngularAcceleration=10, maxAngularVelocity=10, simSpaceAngularAlpha=0, externalAngularVelocity=<2 2 2>}
 }
]
